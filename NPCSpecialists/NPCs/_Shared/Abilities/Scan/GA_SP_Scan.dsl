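; GA_SP_Scan is a specialist NPC ability that periodically sends a scan pulse while the bot is hired.
; A looping timer checks the Converted and blocking tags, then optionally triggers a pre-scan montage event.
; A gameplay cue plays the pulse, then after a delay PulseIndicatorComponent marks for the hirer's team and controller, stopping next tick.
; It then sends a SpecialistAction event and optional speech bubble; BotAlertLevelCheck is defined but never called here.
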(event Ability|EventActivateAbility
  ;[1 1 1] Initialization of variables / Checking if NPC is reviving to cancel this scan ability
  (Ability|CommitAbility)
  (bind _asfort_player_pawn (Utilities|Casting|CastToFortPlayerPawn (Ability|GetAvatarActorfromActorInfo))
    (:then
      (Variables|InternalData|SetBotPawn _asfort_player_pawn)
      (bind _asfort_athena_aibot_controller (Utilities|Casting|CastToFortAthenaAIBotController (Pawn|GetController (Variables|InternalData|GetBotPawn)))
        (:then
          (Variables|InternalData|SetBotController _asfort_athena_aibot_controller)
          (Utilities|Time|SetTimerbyEvent (delegate Custom|SendPulse) (ScalableFloat|GetValueatLevel (Variables|Default|GetPingFrequency)) true))
        (:CastFailed
          (Ability|EndAbility))))
    (:CastFailed
      (Ability|EndAbility))))

(event Custom|SendPulse
  ;[1 1 1] Start Pulse. Checking for hired NPC and his alert level to decide if pulse should be sent. Playing montage before pulse if selected.
  (if (and (GameplayTags|HasMatchingGameplayTag (Variables|InternalData|GetBotPawn) "(TagName=\"Athena.Gameplay.Status.Converted\")") (not (GameplayTags|HasAnyMatchingGameplayTags (Variables|InternalData|GetBotPawn) (Variables|Default|GetTG_PreventAbility))))
    (if (GameplayTags|HasAnyMatchingGameplayTags (Variables|InternalData|GetBotPawn) (GameplayTags|MakeGameplayTagContainerfromArray (Utilities|Array|MakeArray (Variables|Default|GetReviveInteractTag) "(TagName=\"MovementMode.Skydiving\")" "(TagName=\"MovementMode.Parachuting\")")))
      _
      (elif (Variables|Default|GetUsePreScanAnim)
        (Ability|SendGameplayEventtoActor (Variables|InternalData|GetBotPawn) (Variables|InternalData|GetAnimTag) (Utilities|Struct|MakeGameplayEventData "(TagName=\"\")" 0 0 (Utilities|Struct|BreakFortGameplayAbilityMontageInfo (Variables|Default|GetMontageInfoForScan))))))
    (CallFunction|PlayPulseFX)
    (Utilities|FlowControl|Delay (ScalableFloat|GetValueatLevel (Variables|Default|GetDelayBetweenPulseVisualAndMarking)))
    (CallFunction|ActivatePulseIndicator)
    ;[1 1 1] Handle speech bubble behaviors
    (Ability|SendGameplayEventtoActor (Variables|InternalData|GetBotPawn) "(TagName=\"Event.Ability.NPC.SpecialistAction\")")
    (if (Variables|Default|GetTriggerSpeechBubbleOnScan)
      (Class|BPFLAICommonFunctions|PlaySpeechBubble (Variables|InternalData|GetBotPawn) (Variables|Default|GetSpeechBubbleTag) true true))))

(event Ability|EventFailedToActivatePassiveAbility
  (params (ActorInfo :type Gameplay%20Ability%20Actor%20Info%20Structure)))

(event Custom|PlayPulseFX
  ;[1 1 1] Pulse FX and Sound
  (Ability|ExecuteGameplayCueWithParamsOnOwner (Variables|Default|GetGCPulseFXTag) (Ability|GameplayCue|MakeGameplayCueParameters :NormalizedMagnitude 0.0 :RawMagnitude (ScalableFloat|GetValueatLevel (Variables|Default|GetPulseVisualRadius)) :EffectContext (Class|AthenaFunctionLibrary|MakeSimpleFortEffectContext (Variables|InternalData|GetBotPawn) (Variables|InternalData|GetBotPawn)) :Location "0,0,0" :Normal "0,0,0" :Instigator (Variables|InternalData|GetBotPawn) :EffectCauser (Variables|InternalData|GetBotPawn) :SourceObject (Variables|InternalData|GetBotPawn) :GameplayEffectLevel 1 :AbilityLevel 1 :bReplicateLocationWhenUsingMinimalRepProxy false)))

(event Ability|EventOnEndAbility
  (params (bWasCancelled :type Boolean)))

(event Custom|ActivatePulseIndicator
  ;[1 1 1] Activate Pusle Indicator
  (Variables|Default|SetPulseIndicator (Actor|GetComponentbyClass (Variables|InternalData|GetBotPawn) "/Script/FortniteGame.PulseIndicatorComponent"))
  (Variables|InternalData|SetBotOwner (Convert|GetConverterPawn (Variables|/Script/FortniteGame.FortPlayerPawn|GetConvertComponent (Variables|InternalData|GetBotPawn))))
  (Class|PulseIndicatorComponent|SetDisplayTeam (Variables|Default|GetPulseIndicator) (Team|GetActorTeam (Variables|InternalData|GetBotOwner)))
  (bind _asfort_player_controller_athena (Utilities|Casting|CastToFortPlayerControllerAthena (Fortnite|GetFortPlayerControllerfromActor (Variables|InternalData|GetBotOwner)))
    (:then
      (Class|PulseIndicatorComponent|SetIndicatingController (Variables|Default|GetPulseIndicator) _asfort_player_controller_athena))
    (:CastFailed
      _))
  (Class|PulseIndicatorComponent|StartPulseMarkers (Variables|Default|GetPulseIndicator))
  (Utilities|FlowControl|DelayUntilNextTick)
  (Class|PulseIndicatorComponent|StopPulseMarkers (Variables|Default|GetPulseIndicator)))

(fn BotAlertLevelCheck
  (return (Math|Boolean|ORBoolean (Utilities|Enum|Equal(Enum) (Variables|/Script/FortniteAI.FortAthenaAIBotController|GetCurrentAlertLevel (Variables|InternalData|GetBotController)) "Unaware") (Utilities|Enum|Equal(Enum) (Variables|/Script/FortniteAI.FortAthenaAIBotController|GetCurrentAlertLevel (Variables|InternalData|GetBotController)) "Suspicious") (Utilities|Enum|Equal(Enum) (Variables|/Script/FortniteAI.FortAthenaAIBotController|GetCurrentAlertLevel (Variables|InternalData|GetBotController)) "LostTarget"))))
